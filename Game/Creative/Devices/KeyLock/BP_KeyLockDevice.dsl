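; Creative lock device that attaches to a nearby lockable object, such as a door, and lets channel receivers lock, unlock, toggle, open or close it.
; Placement requires a lockable actor; minigame start and end reset the open and lock state from device options on the server.
; Clients show attachment before the game, lock state during it, and device color through custom primitive data.
; Option changes migrate same-channel bindings and update visibility, which drives collision.
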(event EventOnWorldReady
  (Visibility|SetVisibilityComponents (Variables|Default|GetVisibleInGameComponent) (Utilities|Array|MakeArray (Variables|Debug|GetStaticMeshComponent)) true))

(event EventBeginPlay
  ;[1 1 1] [CLIENT ONLY] Bind to the necessary FX events , and set to a valid state the device display
  (Networking|CanExecuteCosmeticEvents
    (:True
      (Visuals|UpdatePreGameDeviceScreen (Utilities|IsValid (Variables|BuildingProp_LockDevice|GetLockableObject) :!class /Script/Engine.KismetSystemLibrary))
      (EventDispatchers|BindEventtoOnLockableActorChange (delegate Custom|OnLockableObjectChangedFX))
      (EventDispatchers|BindEventtoOnLockStateChange (delegate Custom|OnLockStateChangeFX)))))

(event Custom|OnLockableObjectChangedFX
  (params (LockableActorFound :type Actor%20Object%20Reference))
  ;[1 1 1] [CLIENT ONLY] Bind to the necessary FX events , and set to a valid state the device display
  (Visuals|UpdatePreGameDeviceScreen (Utilities|IsValid LockableActorFound :!class /Script/Engine.KismetSystemLibrary)))

(event Custom|OnLockStateChangeFX
  (params (NewLockState :type ELockState%20Enum :default "INVALID"))
  ;[1 1 1] [CLIENT ONLY] Bind to the necessary FX events , and set to a valid state the device display
  (Networking|CanExecuteCosmeticEvents
    (:True
      (Visuals|UpdateonGameDeviceScreen NewLockState))))

(event MinigameStartedDelegate(FortMinigameLogic)
  (CallFunction|HandleMinigameStarted)
  (Gameplay|ResetDevice) ; SERVER ONLY
)

(event MinigameEndedDelegate(FortMinigameLogic)
  (CallFunction|HandleMinigameEnded)
  (Creative|Networking|Authority/Cosmetic
    (:Authority
      (Gameplay|ResetDevice))
    (:Cosmetic
      (Visuals|UpdatePreGameDeviceScreen (Utilities|IsValid (Variables|BuildingProp_LockDevice|GetLockableObject) :!class /Script/Engine.KismetSystemLibrary)))))

(event Custom|HandleUnlock
  (params (Instigator :type Controller%20Object%20Reference))
  ;[1 1 1] [SERVER ONLY] Unlock/Lock Receivers
  (CallFunction|UnlockObject Instigator))

(event Custom|HandleLock
  (params (Instigator :type Controller%20Object%20Reference))
  ;[1 1 1] [SERVER ONLY] Unlock/Lock Receivers
  (CallFunction|LockObject Instigator))

(event Custom|HandleToggleLocked
  (params (Instigator :type Controller%20Object%20Reference))
  ;[1 1 1] [SERVER ONLY] Unlock/Lock Receivers
  (CallFunction|ToggleLockedState Instigator))

(event Custom|HandleOpen
  (params (Instigator :type Controller%20Object%20Reference))
  ;[1 1 1] [SERVER ONLY] Open/Close Receivers
  (CallFunction|Open Instigator))

(event Custom|HandleClose
  (params (Instigator :type Controller%20Object%20Reference))
  ;[1 1 1] [SERVER ONLY] Open/Close Receivers
  (CallFunction|Close Instigator))

(event Custom|HandleToggleOpened
  (params (Instigator :type Controller%20Object%20Reference))
  ;[1 1 1] [SERVER ONLY] Open/Close Receivers
  (CallFunction|ToggleOpenState Instigator))

(event EventPostFinishSpawning
  (Default|MigrateSameChannelInstances)
  (|CallParentFunction))

(event OnAnyPropertyChanged(ToyOptionsComponent)
  (Default|MigrateSameChannelInstances)
  (Visibility|SetVisiblityDuringPhase (Variables|Default|GetVisibleInGameComponent) (select (Variables|KnobSettings|GetbIsVisible) "NewEnumerator3" "NewEnumerator1"))
  (Networking|CanExecuteCosmeticEvents
    (:True
      (Visuals|SetDeviceColor))))

(event OnPropertiesLoaded(ToyOptionsComponent)
  (Default|MigrateSameChannelInstances)
  (Visibility|SetVisiblityDuringPhase (Variables|Default|GetVisibleInGameComponent) (select (Variables|KnobSettings|GetbIsVisible) "NewEnumerator3" "NewEnumerator1"))
  (Networking|CanExecuteCosmeticEvents
    (:True
      (Visuals|SetDeviceColor))))

(event OnVisibilityChanged(VisibleInGameComponent)
  (params (Visible :type Boolean))
  (Gameplay|UpdateCollision Visible))

(event EventDestroyed)

(event EditorPreview|EventSetupForEditorPreview
  ;[1 1 1 1 14] When the device is being moved in editor, manually check for lockable actors and update the visuals accordingly
  (CallFunction|EditorUpdateLockableObject))

(fn AllowInverseScaling
  (return false))

(fn EditorUpdateLockableObject
  (CallFunction|FindandStoreAClosestLockableActor)
  ;[1 1 1 1 13] In editor, the device isn't notified about the lockable object getting destroyed, so we need to call this even if the LockableObject hasn't changed. The result of this function is low-cost anyway (just sets a value on the material)
  ;
  (CallFunction|OnLockableObjectChangedFX (Variables|BuildingProp_LockDevice|GetLockableObject)))

(fn ForceGridAlignment
  (return false))

(fn GetAdditionalSelectableActors
  (return 0))

(fn GetBlockPitchAndRoll
  (return false))

(fn GetBoundsComponents
  (params (SelfActor :type Actor%20Object%20Reference))
  (bind _currentminigame (Variables|/Script/FortniteGame.FortMinigameLogicComponent|GetCurrentMinigame (Variables|Default|GetFortMinigameLogic))
    (:then
      (if (Minigame|IsinProgress _currentminigame)
        (return (Variables|BuildingProp_LockDevice|GetInteractionCollisionComponents))
        (else
          _)))
    (:else
      _))
  (return (Utilities|Array|MakeArray (Variables|Debug|GetStaticMeshComponent))))

(fn GetCollisionStaticMesh
  (return "0, 0, 0" (Variables|/Script/Engine.StaticMeshComponent|GetStaticMesh (Variables|Debug|GetStaticMeshComponent))))

(fn GetMeshComponents
  (return (Utilities|Array|MakeArray (Variables|Debug|GetStaticMeshComponent))))

(fn GetMIDFromMesh
  (params (Mesh :type Static%20Mesh%20Component%20Object%20Reference) (MaterialIndex :type Integer))
  (bind _mid (Material|GetMIDfromMaterial Mesh MaterialIndex (Rendering|Material|GetMaterial Mesh MaterialIndex)))
  (bind _returnvalue (Rendering|Material|GetMaterial Mesh MaterialIndex))
  (bind _asmaterial_instance_dynamic (Utilities|Casting|CastToMaterialInstanceDynamic _returnvalue)
    (:then
      (return _asmaterial_instance_dynamic))
    (:CastFailed
      (Rendering|Material|SetMaterial Mesh MaterialIndex _mid)
      (return _mid))))

(fn GetScaleMax
  (params (DefaultScaleMax :type Float%20%28single-precision%29))
  (return DefaultScaleMax))

(fn GetScaleMin
  (params (DefaultScaleMin :type Float%20%28single-precision%29))
  (return DefaultScaleMin))

(fn GetSpawnedActors
  (return 0))

(fn HasAdditionalSelectableActors
  (return false))

(fn HasAdditionalValidPlacementLogic
  (return true))

(fn IgnoreWorldGridScaleOverride
  (return false))

(fn IsMinigameInProgress
  (bind _fortminigamelogic (Variables|Default|GetFortMinigameLogic))
  (bind _currentminigame (Variables|/Script/FortniteGame.FortMinigameLogicComponent|GetCurrentMinigame _fortminigamelogic))
  (if (Utilities|IsValid _currentminigame :!class /Script/Engine.KismetSystemLibrary)
    (return (Minigame|IsinProgress _currentminigame))
    (else
      (return false))))

(fn IsValidPlacement
  (Utilities|FlowControl|Sequence
    (:then_0
      (bind _returnvalue (CallFunction|FindLockableActor))
      (Utilities|IsValid _returnvalue
        (:Is%20Valid
          (LocalVariables|SetbValidPlacementFound true))
        (:Is%20Not%20Valid
          (LocalVariables|SetbValidPlacementFound false))))
    (:then_1
      (return (LocalVariables|GetbValidPlacementFound)))))

(fn MigrateSameChannelInstances
  (bind _bwasmigrated (UEFNConversionHelper|MigratetoReceiverwhenSameChannel (Variables|Default|GetUnlockWhenReceiveFrom) (Variables|Default|GetLockWhenReceiveFrom) (Variables|Default|GetToggleLockedWhenReceiveFrom)))
  (bind _output_get (LocalVariables|SetbMigrated _bwasmigrated))
  (bind _bwasmigrated_2 (UEFNConversionHelper|MigratetoReceiverwhenSameChannel (Variables|Default|GetOpensWhenReceiveFrom) (Variables|Default|GetCloseWhenReceiveFrom) (Variables|Default|GetToggleOpenedWhenReceiveFrom)))
  (return (or _bwasmigrated_2 _output_get)))

(fn OnPreviewEnd
  (return true))

(fn OnPreviewStart
  (return true))

(fn ResetDevice
  (bind _fortminigamelogic (Variables|Default|GetFortMinigameLogic))
  (bind _currentminigame (Variables|/Script/FortniteGame.FortMinigameLogicComponent|GetCurrentMinigame _fortminigamelogic))
  (Networking|SwitchHasAuthority
    (:Authority
      (switch Utilities|FlowControl|Switch|SwitchonEKeyLock_InitialOpenState (Variables|KnobSettings|GetInitialDoorPosition)
        (:NewEnumerator0
          (CallFunction|Open 0))
        (:NewEnumerator1
          (CallFunction|Close 0)))
      (Utilities|IsValid _currentminigame
        (:Is%20Valid
          (Minigame|IsInProgress _currentminigame
            (:Yes
              (switch Utilities|FlowControl|Switch|SwitchonEKeyLock_LockState (Variables|KnobSettings|GetStartsLocked)
                (:NewEnumerator1
                  (CallFunction|UnlockObject 0))
                (:NewEnumerator0
                  (CallFunction|LockObject 0))))
            (:No
              (CallFunction|UnlockObject 0))))))))

(fn SetDeviceColor
  (Rendering|Material|SetCustomPrimitiveDataFloat (Variables|Debug|GetStaticMeshComponent) (Variables|Visuals|GetTeamColorCPDIndex) (Math|Conversions|ToFloat(Integer) (Variables|KnobSettings|GetColorIndex))))

(fn ShouldBlockPlacement
  (params (MoveTool :type Fort%20Creative%20Move%20Tool%20Object%20Reference))
  (return false))

(fn UpdateCollision
  (params (bComponentsAreVisible :type Boolean))
  (bind _staticmeshcomponent (Variables|Debug|GetStaticMeshComponent))
  (if bComponentsAreVisible
    (Collision|SetCollisionProfileName _staticmeshcomponent (CreativeDevices|GetNamefromCollisionProfile "(Name=\"FortDynamicMeshPhysics\")") true :!class /Script/Engine.PrimitiveComponent)
    (else
      (Collision|SetCollisionProfileName _staticmeshcomponent (CreativeDevices|GetNamefromCollisionProfile "(Name=\"FortCreativeOnlyVolumeManager\")") true :!class /Script/Engine.PrimitiveComponent))))

(fn UpdateOnGameDeviceScreen
  (params (NewLockState :type ELockState%20Enum :default "INVALID"))
  (if (Gameplay|IsMinigameinProgress)
    (switch Utilities|FlowControl|Switch|SwitchonELockState NewLockState
      (:UNLOCKED
        (Rendering|Material|SetCustomPrimitiveDataFloat (Variables|Debug|GetStaticMeshComponent) (Variables|PlacementPreview|GetPlacementStateCPDIndex) 1.0))
      (:LOCKED
        (Rendering|Material|SetCustomPrimitiveDataFloat (Variables|Debug|GetStaticMeshComponent) (Variables|PlacementPreview|GetPlacementStateCPDIndex) 0.0)))))

(fn UpdatePreGameDeviceScreen
  (params (IsObjectAttached :type Boolean))
  (if (Gameplay|IsMinigameinProgress)
    _
    (elif IsObjectAttached
      (Rendering|Material|SetCustomPrimitiveDataFloat (Variables|Debug|GetStaticMeshComponent) (Variables|PlacementPreview|GetPlacementStateAnimationCPDIndex) (Utilities|Time|GetTimeSeconds))
      (Rendering|Material|SetCustomPrimitiveDataFloat (Variables|Debug|GetStaticMeshComponent) (Variables|PlacementPreview|GetPlacementStateCPDIndex) 1.0)
      (else
        (Rendering|Material|SetCustomPrimitiveDataFloat (Variables|Debug|GetStaticMeshComponent) (Variables|PlacementPreview|GetPlacementStateAnimationCPDIndex) (Utilities|Time|GetTimeSeconds))
        (Rendering|Material|SetCustomPrimitiveDataFloat (Variables|Debug|GetStaticMeshComponent) (Variables|PlacementPreview|GetPlacementStateCPDIndex) 0.0)))))

(fn ConstructionScript
  (|CallParentFunction))
